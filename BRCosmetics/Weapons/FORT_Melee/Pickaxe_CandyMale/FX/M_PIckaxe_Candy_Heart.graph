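# Translucent two-sided material for the Candy pickaxe heart effect, drawing a soft, color-tinted emissive shape that pulses like a heartbeat.
# A sine over time samples a curve atlas row, driving both SplineThicken width and a world position offset toward an adjustable object pivot.
# Opacity combines a UV sphere mask, particle alpha, optional camera-angle Fresnel fade and depth fade, which mobile skips.
# Attributes pass through MetalsLUT and emissive adjustment functions; the Sprite function node is unconnected.
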Material /BRCosmetics/Weapons/FORT_Melee/Pickaxe_CandyMale/FX/M_PIckaxe_Candy_Heart
graph {
 properties {
  bUseMaterialAttributes=true
  bUsedWithMeshDeformer=true
  bUsedWithSkeletalMesh=true
  blendMode=BLEND_Translucent
  pixelDepthOffsetMode=PDOM_Legacy
  twoSided=true
  usedShadingModels=Default Lit
 }
 nodes [
  MaterialFunctionCall!(
   Center Location=MaterialFunctionCall() {materialFunction=&ComponentPivotLocation}.Result
   0-1 UVs
   XY scale=Constant() {r=56}
   OverridePivot
   "Normals (Optional)"
  ) {materialFunction=&Sprite}
  n1 = VectorParameter() {parameterName=Color, defaultValue=<rgba 1 1 1 1>}
  n2 = ParticleColor()
  n3 = Multiply(A=@n1.RGB, B=@n2.RGB)
  n4 = TextureCoordinate()
  n5 = Constant() {r=0.5}
  n6 = ScalarParameter() {parameterName=Hardness}
  n7 = Multiply(A=Multiply(A=@n1.A, B=@n2.A), B=Power(Base=SphereMask(A=@n4, B=@n5, Radius=@n5, Hardness=@n6)))
  n8 = Fresnel(ExponentIn=ScalarParameter() {defaultValue=1, group=CameraVectorFade, parameterName=Camera Vector Exponentln}) {baseReflectFraction=0, exponent=0.2}
  n9 = Power(Base=OneMinus(None=Abs(None=@n8)), Exp=ScalarParameter() {defaultValue=1, group=CameraVectorFade, parameterName=Camera vector power}) {constExponent=4}
  n10 = Constant() {r=12}
  n11 = Multiply(
   A=LinearInterpolate(A=@n7, B=Multiply(A=@n7, B=@n9), Alpha=ScalarParameter() {group=CameraVectorFade, parameterName=Camera vector fade})
   B=ShadingPathSwitch(Default=DepthFade(FadeDistance=@n10), Mobile=Constant() {r=1}) {desc=Migrated from FeatureLevelSwitch}
  )
  n12 = Multiply(A=Time(), B=ScalarParameter() {defaultValue=0.45, parameterName=HeartSpeed})
  n13 = CurveAtlasRowParameter(CurveTime=ConstantBiasScale(None=Sine(None=@n12))) {
   atlas=&AthenaCurveAtlas01
   curve=/Game/Athena/LinearColorCurves/CandyMaleCurve
   defaultValue=197
   parameterName=HeartBeatCurve
  }
  n14 = OneMinus(None=@n13.R)
  n15 = LinearInterpolate(B=ScalarParameter() {defaultValue=30, parameterName=Width}, Alpha=@n14) {constB=20}
  n16 = MaterialFunctionCall(WidthBase=@n15, WidthTip=@n15) {materialFunction=&SplineThicken}
  n17 = VectorParameter() {parameterName=PivotOffset, defaultValue=<rgba 0 0 0 1>}
  n18 = Subtract(
   A=Add(
    A=MaterialFunctionCall() {materialFunction=&ObjectPivotPoint}.Object Pivot Location
    B=Transform(None=@n17.RGB) {transformSourceType=TRANSFORMSOURCE_Local}
   )
   B=WorldPosition().XYZ
  )
  n19 = Add(A=LinearInterpolate(A=@n18, B=Normalize(VectorInput=@n18), Alpha=@n14) {constB=0}, B=@n16.WorldPositionOffset)
  MaterialFunctionCall!(In=MaterialFunctionCall(
    Mask=Constant4Vector() {constant=<rgba 1 1 1 1>}.RGBA
    Materials=MakeMaterialAttributes!(
     BaseColor
     Metallic
     Specular
     Roughness
     Anisotropy
     EmissiveColor=@n3
     Opacity=@n11
     OpacityMask
     Normal=@n16.Normal
     Tangent
     WorldPositionOffset=@n19
     SubsurfaceColor
     ClearCoat
     ClearCoatRoughness
     AmbientOcclusion
     Refraction
     CustomizedUVs_0
     CustomizedUVs_1
     CustomizedUVs_2
     CustomizedUVs_3
     CustomizedUVs_4
     CustomizedUVs_5
     CustomizedUVs_6
     CustomizedUVs_7
     PixelDepthOffset
     ShadingModel
     Displacement
    )
   ) {materialFunction=&MF_MetalsLUT}.Result) {materialFunction=/Game/Effects/Niagara/MasterMaterials/Functions/MF_FX_EmissiveAdjustments_MA}
 ]
}
