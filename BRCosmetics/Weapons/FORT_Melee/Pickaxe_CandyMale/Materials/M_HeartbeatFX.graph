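# Surface material for the CandyMale pickaxe, adding a glossy colored heart with an animated heartbeat.
# The FX mask blends heart color, metallic and roughness over the base textures and drives Clearcoat, between the MetalsLUT and TechArtOverrides functions.
# A time-driven sine samples a curve atlas row to blend world position offset between full and normalized vectors toward an offset pivot.
# A sphere-mask undulation falloff is left unconnected.
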Material /BRCosmetics/Weapons/FORT_Melee/Pickaxe_CandyMale/Materials/M_HeartbeatFX
graph {
 properties {
  bUseMaterialAttributes=true
  bUsedWithMeshDeformer=true
  bUsedWithSkeletalMesh=true
  pixelDepthOffsetMode=PDOM_Legacy
  usedShadingModels=Default Lit
 }
 nodes [
  n1 = TextureSampleParameter2D() {parameterName=MaskTexture, texture=/BRCosmetics/Weapons/FORT_Melee/Pickaxe_CandyMale/Textures/T_Pickaxe_CandyMale_FX}
  n2 = Constant4Vector() {constant=<rgba 1 1 1 1>}
  n3 = TextureSample() {samplerType=SAMPLERTYPE_Masks, texture=/BRCosmetics/Weapons/FORT_Melee/Pickaxe_CandyMale/Textures/T_Pickaxe_CandyMale_S}
  n4 = Add(
   A=MaterialFunctionCall() {materialFunction=&ObjectPivotPoint}.Object Pivot Location
   B=Transform(None=VectorParameter() {parameterName=PivotOffset, defaultValue=<rgba 0 5 52 1>}.RGB) {transformSourceType=TRANSFORMSOURCE_Local}
  )
  n5 = WorldPosition()
  n6 = Subtract(A=@n4, B=@n5.XYZ)
  n7 = Multiply(A=Time(), B=ScalarParameter() {defaultValue=0.45, parameterName=HeartSpeed})
  n8 = CurveAtlasRowParameter(CurveTime=ConstantBiasScale(None=Sine(None=@n7))) {
   atlas=&AthenaCurveAtlas01
   curve=/Game/Athena/LinearColorCurves/CandyMaleCurve
   defaultValue=197
   parameterName=HeartBeatCurve
  }
  n9 = OneMinus(None=@n8.R)
  n10 = MakeMaterialAttributes!(
   BaseColor=LinearInterpolate(
    A=TextureSample() {texture=/BRCosmetics/Weapons/FORT_Melee/Pickaxe_CandyMale/Textures/T_Pickaxe_CandyMale_D}.RGB
    B=VectorParameter() {parameterName=HeartColor, defaultValue=<rgba 0.8632 0.0423 0.0561 1>}.RGB
    Alpha=@n1.R
   )
   Metallic=LinearInterpolate(A=@n3.G, B=ScalarParameter() {defaultValue=1, parameterName=HeartMEtallic}, Alpha=@n1.R)
   Specular=@n3.B
   Roughness=LinearInterpolate(A=@n3.R, B=ScalarParameter() {defaultValue=0.1, parameterName=HeartRoughness}, Alpha=@n1.R)
   Anisotropy
   EmissiveColor
   Opacity
   OpacityMask
   Normal=TextureSample() {samplerType=SAMPLERTYPE_Normal, texture=/BRCosmetics/Weapons/FORT_Melee/Pickaxe_CandyMale/Textures/T_Pickaxe_CandyMale_N}.RGB
   Tangent
   WorldPositionOffset=LinearInterpolate(A=@n6, B=Normalize(VectorInput=@n6), Alpha=@n9) {constB=0}
   SubsurfaceColor
   ClearCoat
   ClearCoatRoughness
   AmbientOcclusion
   Refraction
   CustomizedUVs_0
   CustomizedUVs_1
   CustomizedUVs_2
   CustomizedUVs_3
   CustomizedUVs_4
   CustomizedUVs_5
   CustomizedUVs_6
   CustomizedUVs_7
   PixelDepthOffset
   ShadingModel
   Displacement
  )
  MaterialFunctionCall(In=MaterialFunctionCall!(
    Alpha=@n1.R
    In=MaterialFunctionCall(Mask=@n2.RGBA, Materials=@n10) {materialFunction=&MF_MetalsLUT}.Result
   ) {materialFunction=/Game/Weapons/WeaponSkins/MaterialFunction/Clearcoat}.Result) {materialFunction=&MF_TechArtOverrides_accessories}
  n11 = ScalarParameter() {defaultValue=12, parameterName=BeatUndulationRadius}
  n12 = Constant()
  Multiply(A=@n9, B=OneMinus(None=SphereMask(A=@n4, B=@n5.XYZ, Radius=@n11, Hardness=@n12)))
  VertexNormalWS()
 ]
}
