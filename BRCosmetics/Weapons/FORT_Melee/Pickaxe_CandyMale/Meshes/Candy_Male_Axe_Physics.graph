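# Physics asset for the Candy Male pickaxe, a two-handed melee cosmetic with a heart ring bone named as a dynamic part.
# The handle body is a single long capsule.
# The heart ring body combines five angled capsules to approximate a loop shape rather than one solid volume.
# One constraint joins the ring to the handle and sets swing and twist limit angles for the ring's motion.
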PhysicsAsset /BRCosmetics/Weapons/FORT_Melee/Pickaxe_CandyMale/Meshes/Candy_Male_Axe_Physics
bodies {
 Melee_TwoHanded_Handle [{shapeType=Capsule, length=69.361, radius=2.1842, center=<0.6726 -1.0901 -12.789>, rotation=<rot 0.2697 -90 -0.5626>}]
 dyn_heart_ring [
  {shapeType=Capsule, length=16.5012, radius=2.2415, center=<-4.9248 5.3271 -6.1306>, rotation=<rot -29.714 23.0949 -131.6131>}
  {shapeType=Capsule, length=9.2812, radius=2.2415, center=<4.822 -3.385 9.7209>, rotation=<rot -27.2451 18.1475 -99.2532>}
  {shapeType=Capsule, length=20.0922, radius=2.2415, center=<3.2758 -10.8923 -0.3326>, rotation=<rot -19.7746 21.8712 178.898>}
  {shapeType=Capsule, length=8.8594, radius=2.2415, center=<-2.8198 -6.5021 -10.8684>, rotation=<rot -28.0284 17.7191 -74.0153>}
  {shapeType=Capsule, length=17.2057, radius=2.2415, center=<-0.868 6.5536 5.2319>, rotation=<rot -27.2452 18.147 -49.2534>}
 ]
}
constraints [{bone1Name=dyn_heart_ring, bone2Name=Melee_TwoHanded_Handle, swing1LimitDegrees=60, swing2LimitDegrees=60, twistLimitDegrees=90}]
